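# Material function that decodes per-vertex spherical harmonics into a directional thickness value, filed under TechArt refraction.
# Two-band SH coefficients, from vertex color by default, are remapped by an average fit or an optional custom min/max fit chosen by static switch.
# A custom HLSL node evaluates them along the input world direction converted to local space, using Y-up encoding.
# Subtract and multiply inputs normalize the saturated result; a debug output flags out-of-range values red or blue.
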MaterialFunction /Game/Global/MaterialFunctions/TechArt/Refraction/MF_ThicknessSH
graph {nodes [
  n1 = NamedRerouteUsage()
  n2 = Constant3Vector() {constant=<rgba 1 0 0 0>}
  FunctionOutput(None=If(A=@n1, "A > B"=If(A=@n1, "A > B"=@n2, "A == B"=@n1, "A < B"=@n1) {constB=1}, "A < B"=Constant3Vector() {constant=<rgba 0 0 1 0>})) {outputName="Thickness Debug (V3)", sortPriority=20}
  MaterialExpressionStaticSwitch_0 = StaticSwitch(True=NamedRerouteUsage(), False=NamedRerouteUsage(), Value=NamedRerouteUsage())
  MaterialExpressionNamedRerouteUsage_8 = NamedRerouteUsage()
  n3 = Custom!(SH=@MaterialExpressionStaticSwitch_0, Direction=@MaterialExpressionNamedRerouteUsage_8) {
   code="//SH decode,  Direction = LightDir, SH = 2band coefficients. (encoded in Y-up)\r\nDirection = normalize(Direction);\r\nfloat4 shEvalDirection;\r\nshEvalDirection.x = 0.282095;\r\nshEvalDirection.y =0.488603 * Direction.z;\r\nshEvalDirection.z = 0.488603 * Direction.y;\r\nshEvalDirection.w =-0.488603 * Direction.x;\r\nreturn saturate(dot(SH, shEvalDirection));\r\n//return dot(SH, shEvalDirection);"
   description=SphericalHarmonicsEval
   outputType=CMOT_Float1
   inputs [
    {inputName=SH, input {expression=&self:MaterialExpressionStaticSwitch_0}}
    {
     inputName=Direction
     input {expression=&self:MaterialExpressionNamedRerouteUsage_8}
    }
   ]
  }
  n4 = Multiply(A=Subtract(A=@n3, B=NamedRerouteUsage()), B=NamedRerouteUsage())
  FunctionOutput(None=Saturate(None=NamedRerouteDeclaration(None=@n4) {name="SH Thickness Normalized (S)", nodeColor=<rgba 1 0 0.7031 1>})) {outputName="Thickness Normalized (S)"}
  n5 = FunctionInput() {bUsePreviewValueAsDefault=true, inputName=Thickness Fit - Multiply, inputType=FunctionInput_Scalar, sortPriority=60, previewValue=<1 0 0 1>}
  NamedRerouteDeclaration(None=@n5) {name="Thickness Multiply (S)", nodeColor=<rgba 1 0.4453 0 1>}
  MaterialExpressionNamedRerouteUsage_9 = NamedRerouteUsage()
  n6 = Convert(None=@MaterialExpressionNamedRerouteUsage_9) {
   convertInputs [{
     type=Vector4
     defaultValue=<rgba 0 0 0 1>
     expressionInput {
      expression=&self:MaterialExpressionNamedRerouteUsage_9
      inputName=None
      mask=0
      maskA=0
      maskB=0
      maskG=0
      maskR=0
      outputIndex=0
     }
    }]
   convertOutputs [{type=Scalar}, {type=Vector3}]
  }
  n7 = Multiply(A=@n6)
  NamedRerouteDeclaration(None=AppendVector(A=Add(A=@n6), B=Add(A=@n7) {constB=-0.5})) {name="Average Fit (V4)", nodeColor=<rgba 1 0 0.7266 1>}
  n8 = Subtract(A=NamedRerouteUsage(), B=NamedRerouteUsage())
  n9 = NamedRerouteUsage()
  NamedRerouteDeclaration(None=Add(A=Multiply(A=@n8, B=@n9), B=NamedRerouteUsage())) {name="Custom Fit (V4)", nodeColor=<rgba 0 1 0.9531 1>}
  n10 = FunctionInput() {
   bUsePreviewValueAsDefault=true
   inputName=Custom Fit - Coeffs Max
   inputType=FunctionInput_Vector4
   sortPriority=40
   previewValue=<1.2 0.13 0.18 0.13>
  }
  NamedRerouteDeclaration(None=@n10) {name="SH Coeffs Max (V4)", nodeColor=<rgba 0.875 1 0 1>}
  n11 = FunctionInput() {
   bUsePreviewValueAsDefault=true
   inputName=Custom Fit - Coeffs Min
   inputType=FunctionInput_Vector4
   sortPriority=30
   previewValue=<1 -0.16 -0.17 -0.15>
  }
  NamedRerouteDeclaration(None=@n11) {name="SH Coeffs Min (V4)", nodeColor=<rgba 0.8047 0 1 1>}
  NamedRerouteDeclaration(None=Transform(None=FunctionInput() {inputName=Direction WS}) {transformSourceType=TRANSFORMSOURCE_World, transformType=TRANSFORM_Local}) {name="Direction LS (V3)", nodeColor=<rgba 1 0.1641 0 1>}
  n12 = VertexColor()
  n13 = FunctionInput(Preview=AppendVector(A=@n12, B=@n12)) {bUsePreviewValueAsDefault=true, description=Default: VertexColor, inputName=SH Coeffs, inputType=FunctionInput_Vector4, sortPriority=10}
  NamedRerouteDeclaration(None=@n13) {name="SH Coeffs (V4)", nodeColor=<rgba 1 0.0469 0 1>}
  n14 = FunctionInput(Preview=StaticBool()) {bUsePreviewValueAsDefault=true, description=Default: OFF, inputName=Enable Custom Fit, inputType=FunctionInput_StaticBool, sortPriority=20}
  NamedRerouteDeclaration(None=@n14) {name="Enable Custom Fit (SB)", nodeColor=<rgba 0 0.2734 1 1>}
  NamedRerouteDeclaration(None=FunctionInput() {bUsePreviewValueAsDefault=true, inputName=Thickness Fit - Subtract, inputType=FunctionInput_Scalar, sortPriority=50}) {name="Thickness Subtract (S)", nodeColor=<rgba 0 1 0.7656 1>}
 ]}
