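# Material function that composes dielectric volume shading, such as water, into an opaque output of luminance and a constant opacity of 1.
# Luminance behind the surface sums scene luminance, attenuated through a power-adjusted transmittance term, and scattered volume luminance.
# Final luminance adds a term to a masked lerp, combining specular, diffuse and behind-surface lighting.
# The surface opacity mask input moves from visible underwater scene to opaque Lambertian surface, leaving reflections visible.
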MaterialFunction /Game/Global/MaterialFunctions/TechArt/Refraction/MF_DielectricVolumeOpaqueOutput
graph {nodes [
  n1 = NamedRerouteUsage()
  n2 = NamedRerouteUsage()
  n3 = NamedRerouteUsage()
  FunctionOutput(None=Add(A=NamedRerouteUsage(), B=LinearInterpolate(A=@n1, B=@n2, Alpha=@n3))) {outputName="Luminance (V3)"}
  FunctionOutput(None=NamedRerouteUsage() {desc="= 1"}) {outputName="Opacity (S)", sortPriority=10}
  NamedRerouteDeclaration(None=FunctionInput() {bUsePreviewValueAsDefault=true, inputName=Diffuse Luminance, sortPriority=120}) {name="Diffuse Luminance (V3)", nodeColor=<rgba 1 0.8333 0 1>}
  n4 = NamedRerouteUsage()
  n5 = Multiply(A=Power(Base=NamedRerouteUsage(), Exp=Add(A=@n4)), B=NamedRerouteUsage())
  n6 = NamedRerouteUsage()
  n7 = NamedRerouteUsage()
  n8 = NamedRerouteUsage()
  NamedRerouteDeclaration(None=Add(A=Multiply(A=@n5, B=@n6) {desc=Attenuated Scene Luminance}, B=Multiply(A=@n7, B=@n8) {desc=Scattered Luminance})) {name="Luminance Behind Surface (V3)", nodeColor=<rgba 0.1953 0 1 1>}
  NamedRerouteDeclaration(None=FunctionInput() {bUsePreviewValueAsDefault=true, inputName=Volume Luminance, sortPriority=200}) {name="Volume Luminance (V3)", nodeColor=<rgba 0 1 0.1094 1>}
  NamedRerouteDeclaration(None=FunctionInput() {bUsePreviewValueAsDefault=true, inputName=Scene Luminance, sortPriority=250}) {name="Scene Luminance (V3)", nodeColor=<rgba 0.5 0 1 1>}
  NamedRerouteDeclaration(None=FunctionInput() {bUsePreviewValueAsDefault=true, inputName=Specular Luminance, sortPriority=110}) {name="Specular Luminance (V3)", nodeColor=<rgba 0 0.4609 1 1>}
  n9 = FunctionInput() {bUsePreviewValueAsDefault=true, inputName=Surface Transmittance, inputType=FunctionInput_Scalar, sortPriority=150, previewValue=<1 0 0 1>}
  NamedRerouteDeclaration(None=@n9) {name="Surface Transmittance (S)", nodeColor=<rgba 1 0 0.7031 1>}
  n10 = FunctionInput() {
   bUsePreviewValueAsDefault=true
   description="Diffuse opacity of the water surface; 0 : full visibility of the underwater scene + water volume scattering; 1 : the water surface is a fully opaque lambertian reflector; does not affect surface specular/reflection (always visible)"
   inputName=Surface Opacity Mask
   inputType=FunctionInput_Scalar
   sortPriority=140
  }
  NamedRerouteDeclaration(None=@n10) {name="Surface Opacity Mask (S)", nodeColor=<rgba 0.4297 0 1 1>}
  n11 = FunctionInput() {bUsePreviewValueAsDefault=true, inputName=Volume Transmittance, sortPriority=210, previewValue=<1 1 1 1>}
  NamedRerouteDeclaration(None=@n11) {name="Volume Transmittance (V3)", nodeColor=<rgba 0 1 0.6016 1>}
  n12 = FunctionInput() {bUsePreviewValueAsDefault=true, inputName=Is Thin Transmission, inputType=FunctionInput_Scalar, sortPriority=270, previewValue=<1 0 0 1>}
  NamedRerouteDeclaration(None=@n12) {name="Is Thin Transmission (S)", nodeColor=<rgba 0.313 0 1 1>}
  NamedRerouteDeclaration(None=Constant() {r=1}) {name="Opacity Behind Surface (S)", nodeColor=<rgba 0 0.6957 1 1>}
 ]}
