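# Material function that approximates the environment BRDF, giving scalar specular reflectance from NdotV, Roughness, F0 and F90, within the TechArt refraction function set.
# It uses an analytic Lazarov-style polynomial and exponential fit to compute two integral terms, A and B, without a lookup texture.
# These terms weight F0 and F90 into Reflectance, with Transmittance as one minus that.
# Secondary outputs expose F0 and F90 integrals.
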MaterialFunction /Game/Global/MaterialFunctions/TechArt/Refraction/MF_EnvironmentBRDF
graph {nodes [
  FunctionOutput(None=NamedRerouteUsage()) {description=(secondary output), outputName="F0 Integral (S)", sortPriority=110}
  n1 = NamedRerouteUsage()
  n2 = ComponentMask(None=@n1) {r=true}
  n3 = ComponentMask(None=@n1) {g=true}
  n4 = Add(
   A=Multiply(A=NamedRerouteDeclaration(None=@n2) {name="A (S)", nodeColor=<rgba 0 0.6087 1 1>}, B=NamedRerouteUsage())
   B=Multiply(A=NamedRerouteDeclaration(None=@n3) {name="B (S)", nodeColor=<rgba 1 0 0.5739 1>}, B=NamedRerouteUsage())
  )
  FunctionOutput(None=@n4) {outputName="Reflectance (S)"}
  FunctionOutput(None=NamedRerouteUsage()) {description=(secondary output), outputName="F90 Integral (S)", sortPriority=120}
  FunctionOutput(None=Constant()) {outputName=---------------, sortPriority=100}
  FunctionOutput(None=OneMinus(None=@n4)) {description=1 - Reflectance, outputName="Transmittance (S)", sortPriority=10}
  n5 = FunctionInput() {
   bUsePreviewValueAsDefault=true
   description=Incident Reflectance
   inputName=F0
   inputType=FunctionInput_Scalar
   sortPriority=20
   previewValue=<0.04 0 0 1>
  }
  NamedRerouteDeclaration(None=@n5) {name="F0 (S)", nodeColor=<rgba 0.5 1 0 1>}
  n6 = FunctionInput() {
   bUsePreviewValueAsDefault=true
   description=Reflectance at grazing angle
   inputName=F90
   inputType=FunctionInput_Scalar
   sortPriority=30
   previewValue=<1 0 0 1>
  }
  NamedRerouteDeclaration(None=@n6) {name="F90 (S)", nodeColor=<rgba 1 0.5625 0 1>}
  n7 = Constant2Vector() {g=1.04, r=-1.04}
  n8 = Exponential2!(None=Multiply(A=NamedRerouteUsage()) {constB=-9.28})
  n9 = Add(
   A=Multiply(A=NamedRerouteUsage(), B=Constant4Vector() {desc=c0, constant=<rgba -1 -0.0275 -0.572 0.022>}.RGBA)
   B=Constant4Vector() {desc=c1, constant=<rgba 1 0.0425 1.04 -0.04>}.RGBA
  ) {desc=r}
  n10 = ComponentMask(None=@n9) {r=true}
  n11 = Multiply(A=@n10, B=@n10)
  n12 = Add(A=Multiply(A=Min(A=@n8, B=@n11), B=@n10), B=ComponentMask(None=@n9) {g=true}) {desc=a004}
  NamedRerouteDeclaration(None=Add(A=Multiply(A=@n7, B=@n12), B=ComponentMask(None=@n9) {a=true, b=true})) {name="AB (Approx Lazarov) (V2)", nodeColor=<rgba 0.2656 1 0 1>}
  n13 = FunctionInput() {inputName=Roughness, inputType=FunctionInput_Scalar, sortPriority=10, previewValue=<0.2 0 0 1>}
  NamedRerouteDeclaration(None=@n13) {name="Roughness (S)", nodeColor=<rgba 1 0 0.5543 1>}
  NamedRerouteDeclaration(None=FunctionInput() {inputName=NdotV, inputType=FunctionInput_Scalar}) {name="NdotV (S)", nodeColor=<rgba 1 0.1875 0 1>}
 ]}
