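# Material function that models light passing through a homogeneous interior volume for refraction effects, outputting scene transmittance and in-scattered luminance.
# Absorption and scattering sum to extinction, giving Beer-Lambert transmittance over view thickness, with optional extra scene-to-light attenuation for scenes inside the medium.
# Scattering combines sun and sky illuminance through phase functions, using a Schlick approximation of Henyey-Greenstein and an isotropic term, with a mobile shading-path switch for the sun.
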MaterialFunction /Game/Global/MaterialFunctions/TechArt/Refraction/MF_InteriorVolume
graph {nodes [
  n1 = NamedRerouteUsage()
  n2 = NamedRerouteUsage()
  FunctionOutput(None=Multiply(A=NamedRerouteUsage(), B=LinearInterpolate(B=@n1, Alpha=@n2) {constA=1})) {description=RGB attenuation factor of the scene luminance seen through the volume, outputName="Transmittance (V3)", sortPriority=10}
  n3 = NamedRerouteUsage()
  n4 = Divide(A=Multiply(A=OneMinus(None=@n3), B=NamedRerouteUsage()), B=Max(A=NamedRerouteUsage()) {constB=0})
  FunctionOutput(None=Multiply(A=Saturate(None=@n4) {desc=Scattering Ratio}, B=NamedRerouteUsage())) {
   description="Light scattered through the volume that gets back to the camera, based on the phase function's directionality"
   outputName="Scattered Luminance (V3)"
  }
  n5 = FunctionInput() {inputName=Thickness, inputType=FunctionInput_Scalar, sortPriority=240, previewValue=<100 0 0 1>}
  NamedRerouteDeclaration(None=@n5) {name="Thickness (S)", desc=Along the view vector, nodeColor=<rgba 0.2656 0 1 1>}
  NamedRerouteDeclaration(None=ShadingPathSwitch(Default=NamedRerouteUsage(), Mobile=NamedRerouteUsage())) {name="Sun Light Phase Function (S)", nodeColor=<rgba 0 0.0625 1 1>}
  NamedRerouteDeclaration(None=FunctionInput() {inputName=Sky Illuminance, previewValue=<0.5 0.5 1 1>}) {name="Sky Illuminance (V3)", nodeColor=<rgba 0 0.7395 1 1>}
  n6 = Constant() {r=4}
  NamedRerouteDeclaration(None=Divide(B=MaterialFunctionCall(Multiplier=@n6) {materialFunction=&Pi}.Result)) {name="Isotropic Phase Function (S)", nodeColor=<rgba 0.8281 1 0 1>}
  n7 = VertexNormalWS()
  n8 = CameraVectorWS()
  n9 = Multiply(A=Multiply(A=NamedRerouteUsage(), B=DotProduct(A=@n7, B=@n8)) {desc=Depth At Scene Pixel}, B=NamedRerouteUsage())
  NamedRerouteDeclaration(None=Exponential!(None=Multiply(A=@n9) {constB=-1})) {name="Scene To Lights Transmittance (V3)", nodeColor=<rgba 0.8516 0 1 1>}
  n10 = Add(A=NamedRerouteUsage(), B=NamedRerouteUsage())
  n11 = Multiply(
   A=Max(A=NamedRerouteDeclaration(None=@n10) {name="Extinction Coefficient (V3)", nodeColor=<rgba 0.7109 0 1 1>}) {constB=0}
   B=NamedRerouteUsage()
  ) {desc=Optical Depth}
  NamedRerouteDeclaration(None=Exponential(None=Multiply(A=@n11) {constB=-1})) {name="Transmittance (V3)", nodeColor=<rgba 1 0.2109 0 1>}
  n12 = Constant() {r=0.55}
  n13 = Clamp(None=NamedRerouteUsage()) {minDefault=-1}
  n14 = Multiply(A=Multiply(A=@n13, B=@n13), B=@n13)
  n15 = Constant() {r=1.55}
  NamedRerouteDeclaration(None=Subtract(A=Multiply(A=@n15, B=@n13), B=Multiply(A=@n12, B=@n14)) {desc=K}) {name="Henyey-Greenstein K Approx (S)", nodeColor=<rgba 0 1 0.1328 1>}
  n16 = NamedRerouteUsage()
  n17 = NamedRerouteUsage()
  n18 = NamedRerouteUsage()
  n19 = NamedRerouteUsage()
  NamedRerouteDeclaration(None=Add(A=Multiply(A=@n16, B=@n17), B=Multiply(A=@n18, B=@n19))) {name="Incoming Luminance (V3)", nodeColor=<rgba 1 0 0.3516 1>}
  n20 = Clamp(None=NamedRerouteUsage()) {minDefault=-1}
  n21 = Multiply(A=@n20, B=@n20)
  n22 = NamedRerouteUsage()
  n23 = Multiply(A=NamedRerouteUsage()) {constB=-1}
  n24 = Add(A=Multiply(A=@n20, B=DotProduct(A=@n22, B=@n23) {desc=CosTheta})) {desc=SchlickPhaseFactor}
  n25 = Multiply(A=@n24, B=@n24)
  n26 = MaterialFunctionCall(Multiplier=Constant() {r=4}) {materialFunction=&Pi}
  NamedRerouteDeclaration(None=Divide(A=OneMinus(None=@n21), B=Multiply(A=@n25, B=@n26.Result)) {desc=PhaseValue}) {name="Schlick Phase Function (S)", nodeColor=<rgba 1 0 0.1875 1>}
  n27 = FunctionInput(Preview=Multiply(A=CameraVectorWS()) {constB=-1}) {
   bUsePreviewValueAsDefault=true
   description="Refraction vector (or view vector if no interface to traverse); Points away from the surface; Provided vector must be normalized; Default : CameraVectorWS * -1"
   inputName=Incident Vector WS
   sortPriority=100
   previewValue=<0 0 1 1>
  }
  NamedRerouteDeclaration(None=@n27) {name="Incident Vector WS (V3)", nodeColor=<rgba 0 0.2734 1 1>}
  n28 = FunctionInput() {
   bUsePreviewValueAsDefault=true
   description=sigma_a
   inputName=Absorption Coefficient
   sortPriority=210
   previewValue=<0.0021 0.001 0.0004 1>
  }
  NamedRerouteDeclaration(None=Max(A=@n28) {constB=0}) {name="Absorption Coeff (V3)", nodeColor=<rgba 1 0 0.2109 1>}
  NamedRerouteDeclaration(None=FunctionInput() {bUsePreviewValueAsDefault=true, inputName=Is Scene Inside Medium, inputType=FunctionInput_Scalar, sortPriority=280}) {name="Is Scene Inside Medium (S)", nodeColor=<rgba 1 0 0.3281 1>}
  n29 = FunctionInput() {bUsePreviewValueAsDefault=true, description=sigma_s, inputName=Scattering Coefficient, sortPriority=220, previewValue=<0 0 0.0001 1>}
  NamedRerouteDeclaration(None=Max(A=@n29) {constB=0}) {name="Scattering Coeff (V3)", nodeColor=<rgba 0 0.5312 1 1>}
  n30 = FunctionInput() {
   bUsePreviewValueAsDefault=true
   description="Controls directionality of the phase function; -1 : back scattering; 0 : isotropic scattering; 1 : forward scattering"
   inputName=Phase G
   inputType=FunctionInput_Scalar
   sortPriority=230
  }
  NamedRerouteDeclaration(None=@n30) {name="Phase G (S)", nodeColor=<rgba 0.8516 1 0 1>}
  n31 = FunctionInput() {description=Normalized vector to the light, inputName=Sun Light Vector WS, sortPriority=20, previewValue=<0 0 1 1>}
  NamedRerouteDeclaration(None=@n31) {name="Sun Light Vector WS (V3)", desc=To the light, nodeColor=<rgba 0 1 0.625 1>}
  NamedRerouteDeclaration(None=FunctionInput() {inputName=Sun Illuminance, sortPriority=10, previewValue=<1 1 0.5 1>}) {name="Sun Illuminance (V3)", nodeColor=<rgba 1 0.7978 0 1>}
 ]}
