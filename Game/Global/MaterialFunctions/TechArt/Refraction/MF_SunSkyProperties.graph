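# Material function that outputs sun illuminance, sky illuminance and sun direction from the scene's Sky Atmosphere and Sky Light.
# Sun values come from the atmosphere light's ground illuminance and direction.
# Sky illuminance samples the skylight environment map at full roughness, by default along the pixel world normal, times Pi to convert luminance to illuminance.
# Optional inputs scale and desaturate each illuminance, with multipliers clamped non-negative and desaturation saturated to 0–1.
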MaterialFunction /Game/Global/MaterialFunctions/TechArt/Refraction/MF_SunSkyProperties
graph {nodes [
  n1 = SkyAtmosphereLightIlluminanceOnGround()
  n2 = NamedRerouteUsage()
  FunctionOutput(None=Desaturation(None=Multiply(A=@n1, B=@n2), Fraction=NamedRerouteUsage())) {outputName="Sun Illuminance (V3)", sortPriority=10}
  n3 = NamedRerouteUsage()
  n4 = Multiply(
   A=SkyLightEnvMapSample(Direction=NamedRerouteUsage(), Roughness=Constant() {r=1})
   B=MaterialFunctionCall() {materialFunction=&Pi}.Result
  ) {desc=Luminance to illuminance}
  FunctionOutput(None=Desaturation(None=Multiply(A=@n3, B=@n4), Fraction=NamedRerouteUsage())) {outputName="Sky Illuminance (V3)"}
  FunctionOutput(None=SkyAtmosphereLightDirection()) {outputName="Sun Direction (V3)", sortPriority=20}
  n5 = FunctionInput() {bUsePreviewValueAsDefault=true, inputName=Sun Illuminance Mult, inputType=FunctionInput_Scalar, sortPriority=30, previewValue=<1 0 0 1>}
  NamedRerouteDeclaration(None=Max(A=@n5) {constB=0}) {name="Sun Illuminance Mult (S)", nodeColor=<rgba 1 0.7734 0 1>}
  n6 = FunctionInput() {bUsePreviewValueAsDefault=true, inputName=Sky Illuminance Mult, inputType=FunctionInput_Scalar, sortPriority=10, previewValue=<1 0 0 1>}
  NamedRerouteDeclaration(None=Max(A=@n6) {constB=0}) {name="Sky Illuminance Mult (S)", nodeColor=<rgba 0.4531 0 1 1>}
  NamedRerouteDeclaration(None=FunctionInput(Preview=PixelNormalWS()) {bUsePreviewValueAsDefault=true, inputName=Normal WS, previewValue=<0 0 1 1>}) {name="Normal WS (V3)", nodeColor=<rgba 0 0.5913 1 1>}
  NamedRerouteDeclaration(None=Saturate(None=FunctionInput() {bUsePreviewValueAsDefault=true, inputName=Sun Desaturation, inputType=FunctionInput_Scalar, sortPriority=40})) {name="Sun Desaturation (S)", nodeColor=<rgba 0 1 0.1328 1>}
  NamedRerouteDeclaration(None=Saturate(None=FunctionInput() {bUsePreviewValueAsDefault=true, inputName=Sky Desaturation, inputType=FunctionInput_Scalar, sortPriority=20})) {name="Sky Desaturation (S)", nodeColor=<rgba 1 0 0.1875 1>}
 ]}
