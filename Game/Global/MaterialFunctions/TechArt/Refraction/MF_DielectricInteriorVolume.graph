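# Material function that shades a dielectric's refracted interior as a participating volume, outputting luminance for opaque materials and luminance plus opacity for alpha blending.
# IOR-derived Fresnel and roughness feed an environment BRDF for surface transmittance.
# MF_InteriorVolume computes scattering and transmittance from density-scaled coefficients, phase, thickness and sun/sky lighting, which defaults to MF_SunSkyProperties.
# Two output functions composite these with scene luminance and interior emission, with haze for alpha blending only.
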MaterialFunction /Game/Global/MaterialFunctions/TechArt/Refraction/MF_DielectricInteriorVolume
graph {nodes [
  FunctionInput() {bUsePreviewValueAsDefault=true, inputName=//---- ALPHA-BLENDING ONLY ----//, inputType=FunctionInput_Scalar, sortPriority=200}
  FunctionInput() {bUsePreviewValueAsDefault=true, inputName=//----------- LIGHTING -----------//, inputType=FunctionInput_Scalar, sortPriority=300}
  n1 = MaterialFunctionCall!(
   Specular Luminance
   Diffuse Luminance
   Surface Opacity Mask=NamedRerouteUsage()
   Surface Transmittance=NamedRerouteUsage()
   Volume Luminance=NamedRerouteUsage()
   Volume Transmittance=NamedRerouteUsage()
   Scene Luminance=NamedRerouteUsage()
   Haze Amount=NamedRerouteUsage()
   Is Thin Transmission=NamedRerouteUsage()
  ) {materialFunction=&/MF_DielectricVolumeAlphaBlendOutput}
  FunctionOutput(None=@n1."Opacity (S)") {outputName="Material Opacity - AlphaBlending (S)", sortPriority=20}
  FunctionOutput(None=@n1."Luminance (V3)") {outputName="Refraction Result - AlphaBlending (V3)", sortPriority=10}
  FunctionOutput(None=MaterialFunctionCall!(
    Specular Luminance
    Diffuse Luminance
    Surface Opacity Mask=NamedRerouteUsage()
    Surface Transmittance=NamedRerouteUsage()
    Volume Luminance=NamedRerouteUsage()
    Volume Transmittance=NamedRerouteUsage()
    Scene Luminance=NamedRerouteUsage()
    Is Thin Transmission=NamedRerouteUsage()
   ) {materialFunction=&/MF_DielectricVolumeOpaqueOutput}."Luminance (V3)") {outputName="Refraction Result - Opaque (V3)"}
  n2 = FunctionInput() {bUsePreviewValueAsDefault=true, inputName=Volume Density, inputType=FunctionInput_Scalar, sortPriority=80, previewValue=<1 0 0 1>}
  n3 = Multiply(A=Max(A=@n2) {constB=0}) {constA=0.01, constB=0.01}
  NamedRerouteDeclaration(None=Multiply(A=FunctionInput() {bUsePreviewValueAsDefault=true, inputName=Absorption Coeffs, sortPriority=90}, B=@n3)) {name="Volume - Absorption Coeffs (V3)", nodeColor=<rgba 1 0 0.8672 1>}
  NamedRerouteDeclaration(None=Multiply(A=@n3, B=FunctionInput() {bUsePreviewValueAsDefault=true, inputName=Scattering Coeffs, sortPriority=100})) {name="Volume - Scattering Coeffs (V3)", nodeColor=<rgba 0 1 0.8125 1>}
  n4 = FunctionInput() {bUsePreviewValueAsDefault=true, inputName=Scattering Phase, inputType=FunctionInput_Scalar, sortPriority=110}
  NamedRerouteDeclaration(None=Multiply(A=Clamp(None=@n4) {minDefault=-1}) {constB=0.938}) {name="Volume - PhaseG (S)", nodeColor=<rgba 1 0.4922 0 1>}
  NamedRerouteDeclaration(None=FunctionInput() {bUsePreviewValueAsDefault=true, inputName=Haze Amount, inputType=FunctionInput_Scalar, sortPriority=210}) {name="Haze Amount  (S)", nodeColor=<rgba 0.3359 0 1 1>}
  NamedRerouteDeclaration(None=FunctionInput(Preview=PixelNormalWS()) {inputName=Normal WS, previewValue=<0 0 1 1>}) {name="Normal WS (V3)", nodeColor=<rgba 0 0.5913 1 1>}
  NamedRerouteDeclaration(None=FunctionInput() {bUsePreviewValueAsDefault=true, inputName=Surface Opacity, inputType=FunctionInput_Scalar, sortPriority=20}) {name="Surface Opacity (S)", nodeColor=<rgba 0 0.6719 1 1>}
  n5 = FunctionInput() {bUsePreviewValueAsDefault=true, inputName=Roughness, inputType=FunctionInput_Scalar, sortPriority=10, previewValue=<0.25 0 0 1>}
  NamedRerouteDeclaration(None=Max(A=Saturate(None=@n5)) {constB=0.0156}) {name="Roughness (S)", nodeColor=<rgba 0.1719 0 1 1>}
  NamedRerouteDeclaration(None=FunctionInput() {bUsePreviewValueAsDefault=true, inputName=Interior Emission, sortPriority=60}) {name="Interior Emission (V3)", nodeColor=<rgba 0.8047 1 0 1>}
  n6 = DotProduct(A=NamedRerouteUsage(), B=CameraVectorWS())
  NamedRerouteDeclaration(None=MaterialFunctionCall!(NdotV=Saturate(None=@n6), Roughness=NamedRerouteUsage(), F0=NamedRerouteUsage(), F90=NamedRerouteUsage()) {materialFunction=&/MF_EnvironmentBRDF}."Transmittance (S)") {name="Surface Transmittance (S)", nodeColor=<rgba 1 0 0.3984 1>}
  n7 = NamedRerouteUsage()
  n8 = Divide(A=Subtract(A=@n7), B=Add(A=@n7)) {desc=F0Sqrt}
  NamedRerouteDeclaration(None=Multiply(A=@n8, B=@n8)) {name="Refraction - F0 (S)", nodeColor=<rgba 0 1 0.3203 1>}
  n9 = NamedRerouteUsage()
  NamedRerouteDeclaration(None=Saturate(None=Multiply(A=@n9) {constB=50})) {name="Refraction - F90 (S)", nodeColor=<rgba 1 0 0 1>}
  NamedRerouteDeclaration(None=FunctionInput() {bUsePreviewValueAsDefault=true, inputName=Volume Thickness, inputType=FunctionInput_Scalar, sortPriority=70}) {name="Volume Thickness (S)", nodeColor=<rgba 1 0 0.5156 1>}
  NamedRerouteDeclaration(None=Saturate(None=FunctionInput() {bUsePreviewValueAsDefault=true, inputName=Is Scene Inside Volume, inputType=FunctionInput_Scalar, sortPriority=120})) {name="Volume - Is Scene Inside Medium (S)", nodeColor=<rgba 0 1 0.625 1>}
  NamedRerouteDeclaration(None=FunctionInput() {inputName=Scene Luminance, sortPriority=50, previewValue=<1 1 1 1>}) {name="Scene Luminance (V3)", nodeColor=<rgba 0 0.4375 1 1>}
  n10 = MaterialFunctionCall!(Normal WS, Sky Illuminance Mult, Sky Desaturation, Sun Illuminance Mult, Sun Desaturation) {materialFunction=&/MF_SunSkyProperties}
  n11 = FunctionInput(Preview=@n10."Sun Direction (V3)") {bUsePreviewValueAsDefault=true, inputName=Sun Direction, sortPriority=330, previewValue=<1 1 1 1>}
  NamedRerouteDeclaration(None=@n11) {name="Lighting - Sun Direction (V3)", nodeColor=<rgba 0 0.0625 1 1>}
  n12 = MaterialFunctionCall!(
   Sky Illuminance=NamedRerouteUsage()
   Sun Illuminance=NamedRerouteUsage()
   Sun Light Vector WS=NamedRerouteUsage()
   Incident Vector WS=NamedRerouteUsage()
   Absorption Coefficient=NamedRerouteUsage()
   Scattering Coefficient=NamedRerouteUsage()
   Phase G=NamedRerouteUsage()
   Thickness=NamedRerouteUsage()
   Is Scene Inside Medium=NamedRerouteUsage()
  ) {materialFunction=&/MF_InteriorVolume}
  NamedRerouteDeclaration(None=@n12."Scattered Luminance (V3)") {name="Volume Scattered Luminance (V3)", nodeColor=<rgba 1 0 0.8906 1>}
  NamedRerouteDeclaration(None=@n12."Transmittance (V3)") {name="Volume Transmittance (V3)", nodeColor=<rgba 0 1 0.7891 1>}
  n13 = FunctionInput(Preview=@n10."Sun Illuminance (V3)") {
   bUsePreviewValueAsDefault=true
   description=Only used for simulated colored transmittance when not sampling SceneColor
   inputName=Sun Illuminance
   sortPriority=320
   previewValue=<1 1 1 1>
  }
  NamedRerouteDeclaration(None=@n13) {name="Lighting - Sun Illuminance (V3)", nodeColor=<rgba 0.7344 0 1 1>}
  n14 = FunctionInput(Preview=@n10."Sky Illuminance (V3)") {
   bUsePreviewValueAsDefault=true
   description=Only used for simulated colored transmittance when not sampling SceneColor
   inputName=Sky Illuminance
   sortPriority=310
   previewValue=<1 1 1 1>
  }
  NamedRerouteDeclaration(None=@n14) {name="Lighting - Sky Illuminance (V3)", nodeColor=<rgba 0 1 0.4141 1>}
  n15 = FunctionInput() {bUsePreviewValueAsDefault=true, inputName=Refraction IOR, inputType=FunctionInput_Scalar, sortPriority=40, previewValue=<1.5 0 0 1>}
  NamedRerouteDeclaration(None=Clamp(None=@n15) {maxDefault=5, minDefault=1}) {name="Refraction IOR (S)", nodeColor=<rgba 1 0.0938 0 1>}
  NamedRerouteDeclaration(None=Add(A=NamedRerouteUsage(), B=NamedRerouteUsage())) {name="Volume Luminance (V3)", nodeColor=<rgba 1 0.3281 0 1>}
  NamedRerouteDeclaration(None=Max(A=FunctionInput() {bUsePreviewValueAsDefault=true, inputName=Thin Transmission Exp, inputType=FunctionInput_Scalar, sortPriority=130}) {constB=0}) {name="Refraction - Thin Transmission (S)", nodeColor=<rgba 0 0.2266 1 1>}
  n16 = FunctionInput(Preview=Multiply(A=CameraVectorWS()) {constB=-1}) {bUsePreviewValueAsDefault=true, description=Default : CameraVectorWS * -1, inputName=Refraction Vector WS, sortPriority=30}
  NamedRerouteDeclaration(None=@n16) {name="Refraction Vector WS (V3)", nodeColor=<rgba 0.5469 1 0 1>}
 ]}
