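# MF_DistanceToRefractedScene is a material function that returns the depth gap between the current pixel and the scene behind it.
# It subtracts pixel depth from scene depth sampled at an input viewport UV, which defaults to the pixel's own UV.
# A shading path switch makes mobile ignore that UV input.
# The result is clamped between zero and a Max Distance input, covering empty backgrounds like sky.
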MaterialFunction /Game/Global/MaterialFunctions/TechArt/Refraction/MF_DistanceToRefractedScene
graph {
 properties {previewBlendMode=BLEND_Translucent}
 nodes [
  n1 = FunctionInput(Preview=ScreenPosition().ViewportUV) {
   bUsePreviewValueAsDefault=true
   description="Defaults to viewport UV of the current pixel; Not used for mobile platforms"
   inputName=Viewport UV
   inputType=FunctionInput_Vector2
  }
  n2 = Subtract(A=ShadingPathSwitch(Default=SceneDepth(UVs=@n1).R, Mobile=SceneDepth().R), B=PixelDepth().R)
  n3 = FunctionInput() {
   bUsePreviewValueAsDefault=true
   description="Distance returned when there's no geo behind the sampled pixel (eg. sky)"
   inputName=Max Distance
   inputType=FunctionInput_Scalar
   sortPriority=10
   previewValue=<10000 0 0 1>
  }
  FunctionOutput(None=Min(A=Max(A=@n2) {constB=0}, B=@n3)) {outputName="Distance (S)"}
 ]
}
