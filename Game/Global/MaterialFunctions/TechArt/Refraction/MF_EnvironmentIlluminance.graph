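# Material function that outputs environment illuminance for a world-space direction, defaulting to the reflection vector.
# Sky illuminance samples the sky light environment map by roughness, with desaturation and brightness controls.
# An optional ground lat-long texture, mapped via atan2 and fast arccosine with rotation aligned to the SkyLight, receives exponent and brightness shaping.
# A static switch enables a smoothstep blend between sky and ground around the horizon.
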MaterialFunction /Game/Global/MaterialFunctions/TechArt/Refraction/MF_EnvironmentIlluminance
graph {nodes [
  n1 = NamedRerouteUsage()
  n2 = NamedRerouteUsage()
  n3 = SmoothStep(Value=ComponentMask(None=NamedRerouteUsage()) {b=true}) {constMax=0.1, constMin=-0.1}
  n4 = FunctionInput(Preview=StaticBool()) {bUsePreviewValueAsDefault=true, description=Default : Off, inputName=Use Ground Map, inputType=FunctionInput_StaticBool, sortPriority=40}
  FunctionOutput(None=StaticSwitch(True=LinearInterpolate(A=@n1, B=@n2, Alpha=@n3), False=NamedRerouteUsage(), Value=@n4)) {outputName="Environment Illuminance (V3)"}
  n5 = TextureSample(
   UVs=NamedRerouteUsage()
   Tex=FunctionInput(Preview=TextureObject() {samplerType=SAMPLERTYPE_LinearColor, texture=/Game/Global/MaterialFunctions/TechArt/Refraction/Shared/T_Latlong_Sky360_256_FN}) {bUsePreviewValueAsDefault=true, inputName=Ground LatLong Map, inputType=FunctionInput_Texture2D, sortPriority=50}
  ) {
   samplerSource=SSM_Wrap_WorldGroupSettings
   samplerType=SAMPLERTYPE_LinearColor
   texture=/Game/Global/Textures/Default/Blanks/T_BlackLinearColor
  }
  n6 = FunctionInput() {bUsePreviewValueAsDefault=true, inputName=Ground Exp, inputType=FunctionInput_Scalar, sortPriority=70, previewValue=<1 0 0 1>}
  n7 = FunctionInput() {bUsePreviewValueAsDefault=true, inputName=Ground Brightness, inputType=FunctionInput_Scalar, sortPriority=80, previewValue=<1 0 0 1>}
  NamedRerouteDeclaration(None=Multiply(A=Power(Base=@n5.RGB, Exp=@n6), B=@n7)) {name="Ground Illuminance (V3)", nodeColor=<rgba 0.6406 1 0 1>}
  n8 = FunctionInput() {bUsePreviewValueAsDefault=true, inputName=Roughness, inputType=FunctionInput_Scalar, sortPriority=10, previewValue=<0.25 0 0 1>}
  n9 = SkyLightEnvMapSample(Direction=NamedRerouteUsage(), Roughness=Saturate(None=@n8))
  n10 = FunctionInput() {bUsePreviewValueAsDefault=true, inputName=Sky Desaturation, inputType=FunctionInput_Scalar, sortPriority=20}
  n11 = FunctionInput() {bUsePreviewValueAsDefault=true, inputName=Sky Brightness, inputType=FunctionInput_Scalar, sortPriority=30, previewValue=<1 0 0 1>}
  NamedRerouteDeclaration(None=Multiply(A=Desaturation(None=@n9, Fraction=@n10), B=@n11)) {name="Sky Illuminance (V3)", nodeColor=<rgba 1 0 0.9609 1>}
  MaterialExpressionNamedRerouteDeclaration_42 = NamedRerouteDeclaration(None=FunctionInput(Preview=ReflectionVectorWS()) {inputName=Direction WS}) {name="Direction WS (V3)", nodeColor=<rgba 1 0.4688 0 1>}
  MaterialExpressionConvert_0 = Convert(None=@MaterialExpressionNamedRerouteDeclaration_42) {
   nodeName=Break Float3
   convertInputs [{
     type=Vector3
     expressionInput {
      expression=&self:MaterialExpressionNamedRerouteDeclaration_42
      inputName=None
      mask=0
      maskA=0
      maskB=0
      maskG=0
      maskR=0
      outputIndex=0
     }
    }]
   convertOutputs [{type=Scalar}, {type=Scalar}, {type=Scalar}]
  }
  n12 = Custom!(x=@MaterialExpressionConvert_0, y=@MaterialExpressionConvert_0) {
   code="return (atan2(y,x));"
   description=atan2
   outputType=CMOT_Float1
   inputs [
    {inputName=x, input {expression=&self:MaterialExpressionConvert_0}}
    {
     inputName=y
     input {expression=&self:MaterialExpressionConvert_0, outputIndex=1}
    }
   ]
  }
  n13 = Add(
   A=Add(A=Divide(A=@n12) {constB=6.2832, desc="/ (2*pi)"}) {constB=-0.25, desc="base orientation\r\n(matches SkyLight)"}
   B=FunctionInput() {bUsePreviewValueAsDefault=true, inputName=Group Map Rotation, inputType=FunctionInput_Scalar, sortPriority=60}
  )
  n14 = OneMinus(None=Divide(A=ArccosineFast!(None=@MaterialExpressionConvert_0)) {constB=3.1416, desc=/ pi})
  NamedRerouteDeclaration(None=AppendVector(A=Frac(None=@n13), B=Multiply(A=@n14) {constB=-1, desc=Flip V})) {name="Latlong UV (V2)", nodeColor=<rgba 1 0 0.0234 1>}
 ]}
