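# Material function that computes a refraction vector from a view vector, a normalized surface or microfacet normal and a relative IOR (environment over medium).
# It applies the vector form of Snell's law, using saturated N·V as the incident cosine to derive k, the squared transmitted cosine.
# It outputs the vector pointing away from the interface, plus a total internal reflection flag derived from a sign test.
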MaterialFunction /Game/Global/MaterialFunctions/TechArt/Refraction/MF_Refract
graph {nodes [
  n1 = Sign(None=NamedRerouteUsage())
  FunctionOutput(None=OneMinus(None=Saturate(None=@n1))) {outputName="Is Total Internal Reflection (S)", sortPriority=10}
  FunctionOutput(None=NamedRerouteUsage()) {description=Points away from the interface, outputName="Refraction Vector (V3)"}
  n2 = FunctionInput() {
   description="environmentIOR / mediumIOR (usually 1 / mediumIOR, assuming air)"
   inputName="Relative IOR (Eta)"
   inputType=FunctionInput_Scalar
   sortPriority=30
   previewValue=<0.75 0 1 1>
  }
  NamedRerouteDeclaration(None=@n2) {name="Eta (S)", nodeColor=<rgba 0 0.4141 1 1>}
  n3 = NamedRerouteUsage()
  n4 = NamedRerouteUsage()
  n5 = NamedRerouteDeclaration(None=Saturate(None=DotProduct(A=@n3, B=@n4))) {name="CosTheta_i (S)", nodeColor=<rgba 0.7344 1 0 1>}
  n6 = OneMinus(None=Multiply(A=@n5, B=@n5))
  n7 = NamedRerouteUsage()
  n8 = Multiply(A=@n7, B=@n7)
  NamedRerouteDeclaration(None=OneMinus(None=Multiply(A=@n6, B=@n8)) {desc="k = cos^2(Theta_t)"}) {name="k (S)", nodeColor=<rgba 0.9453 0 1 1>}
  n9 = Subtract(A=Multiply(A=NamedRerouteUsage(), B=NamedRerouteUsage()), B=SquareRoot(None=NamedRerouteUsage()))
  n10 = NamedRerouteUsage()
  n11 = NamedRerouteUsage()
  n12 = NamedRerouteUsage()
  NamedRerouteDeclaration(None=Subtract(A=Multiply(A=@n9, B=@n10), B=Multiply(A=@n11, B=@n12))) {name="Refraction Vector (V3)", nodeColor=<rgba 0 0.625 1 1>}
  n13 = FunctionInput() {
   description="Surface normal (N) for surface context, microfacet normal/half-vector (H) for microfacet context. Provided normal must be normalized"
   inputName=Normal
   sortPriority=20
   previewValue=<0 0 1 1>
  }
  NamedRerouteDeclaration(None=@n13) {name="Normal (V3)", nodeColor=<rgba 0 0.2734 1 1>}
  n14 = FunctionInput() {description="Normalized view vector; Default : CameraVectorWS", inputName=View Vector, sortPriority=10, previewValue=<0 0 1 1>}
  NamedRerouteDeclaration(None=@n14) {name="View Vector (V3)", nodeColor=<rgba 1 0 0.6562 1>}
 ]}
