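# Material function that builds a binary refraction mask, flagging where distorted screen UVs land on scene behind the current surface rather than in front of it.
# It samples SceneDepth at four diagonal half-texel offsets around the input Distorted Viewport UV.
# Convert nodes pack the samples into a vector to subtract PixelDepth at once.
# The smallest difference passes through Sign and Saturate, giving 1 only when all four samples are farther than the pixel.
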MaterialFunction /Game/Global/MaterialFunctions/TechArt/Refraction/MF_RefractionMask
graph {expressions {
  MaterialExpressionAdd_1 {
   class=MaterialExpressionAdd
   inputs [
    A {from=MaterialExpressionMultiply_2}
    B {from=MaterialExpressionFunctionInput_0}
   ]
  }
  MaterialExpressionAdd_2 {
   class=MaterialExpressionAdd
   inputs [
    A {from=MaterialExpressionMultiply_5}
    B {from=MaterialExpressionFunctionInput_0}
   ]
  }
  MaterialExpressionAdd_3 {
   class=MaterialExpressionAdd
   inputs [
    A {from=MaterialExpressionMultiply_6}
    B {from=MaterialExpressionFunctionInput_0}
   ]
  }
  MaterialExpressionAdd_51 {
   class=MaterialExpressionAdd
   inputs [
    A {from=MaterialExpressionMultiply_100}
    B {from=MaterialExpressionFunctionInput_0}
   ]
  }
  MaterialExpressionConstant2Vector_5 {class=MaterialExpressionConstant2Vector, properties {g=0.5, r=-0.5}}
  MaterialExpressionConstant2Vector_6 {class=MaterialExpressionConstant2Vector, properties {g=0.5, r=0.5}}
  MaterialExpressionConstant2Vector_7 {class=MaterialExpressionConstant2Vector, properties {g=-0.5, r=0.5}}
  MaterialExpressionConstant2Vector_8 {class=MaterialExpressionConstant2Vector, properties {g=-0.5, r=-0.5}}
  MaterialExpressionConvert_3 {
   class=MaterialExpressionConvert
   inputs [None {from=MaterialExpressionPixelDepth_1, output=R}]
   properties {
    convertInputs [{
      type=Scalar
      expressionInput {
       expression=&self:MaterialExpressionPixelDepth_1
       inputName=None
       mask=1
       maskA=0
       maskB=0
       maskG=0
       maskR=1
       outputIndex=0
      }
     }]
    convertOutputs [{type=Vector4, defaultValue=<rgba 0 0 0 1>}]
   }
  }
  MaterialExpressionConvert_4 {
   class=MaterialExpressionConvert
   inputs [
    None {from=MaterialExpressionSceneTexture_0, output=Color}
    None {from=MaterialExpressionSceneTexture_1, output=Color}
    None {from=MaterialExpressionSceneTexture_2, output=Color}
    None {from=MaterialExpressionReroute_3}
   ]
   properties {
    convertInputs [
     {
      type=Scalar
      expressionInput {
       expression=&self:MaterialExpressionSceneTexture_0
       inputName=None
       mask=1
       maskA=1
       maskB=1
       maskG=1
       maskR=1
       outputIndex=0
      }
     }
     {
      type=Scalar
      expressionInput {
       expression=&self:MaterialExpressionSceneTexture_1
       inputName=None
       mask=1
       maskA=1
       maskB=1
       maskG=1
       maskR=1
       outputIndex=0
      }
     }
     {
      type=Scalar
      expressionInput {
       expression=&self:MaterialExpressionSceneTexture_2
       inputName=None
       mask=1
       maskA=1
       maskB=1
       maskG=1
       maskR=1
       outputIndex=0
      }
     }
     {
      type=Scalar
      expressionInput {
       expression=&self:MaterialExpressionReroute_3
       inputName=None
       mask=0
       maskA=0
       maskB=0
       maskG=0
       maskR=0
       outputIndex=0
      }
     }
    ]
    convertOutputs [{type=Vector4, defaultValue=<rgba 0 0 0 1>}]
   }
  }
  MaterialExpressionConvert_5 {
   class=MaterialExpressionConvert
   inputs [None {from=MaterialExpressionSubtract_2}]
   properties {
    convertInputs [{
      type=Vector4
      defaultValue=<rgba 0 0 0 1>
      expressionInput {
       expression=&self:MaterialExpressionSubtract_2
       inputName=None
       mask=0
       maskA=0
       maskB=0
       maskG=0
       maskR=0
       outputIndex=0
      }
     }]
    convertOutputs [{type=Scalar}, {type=Scalar}, {type=Scalar}, {type=Scalar}]
   }
  }
  MaterialExpressionFunctionInput_0 {
   class=MaterialExpressionFunctionInput
   inputs [Preview {from=MaterialExpressionScreenPosition_0, output=ViewportUV}]
   properties {inputName=Distorted Viewport UV, inputType=FunctionInput_Vector2}
  }
  MaterialExpressionFunctionOutput_1 {
   class=MaterialExpressionFunctionOutput
   inputs [None {from=MaterialExpressionSaturate_1, output=""}]
   properties {outputName="Distortion Mask (S)"}
  }
  MaterialExpressionMin_3 {
   class=MaterialExpressionMin
   inputs [
    A {from=MaterialExpressionConvert_5, output=""}
    B {from=MaterialExpressionConvert_5, output=""}
   ]
   properties {}
  }
  MaterialExpressionMin_4 {
   class=MaterialExpressionMin
   inputs [
    A {from=MaterialExpressionMin_3, output=""}
    B {from=MaterialExpressionConvert_5, output=""}
   ]
   properties {}
  }
  MaterialExpressionMin_5 {
   class=MaterialExpressionMin
   inputs [
    A {from=MaterialExpressionMin_4, output=""}
    B {from=MaterialExpressionConvert_5, output=""}
   ]
   properties {}
  }
  MaterialExpressionMultiply_100 {
   class=MaterialExpressionMultiply
   inputs [
    A {from=MaterialExpressionSceneTexelSize_7}
    B {from=MaterialExpressionConstant2Vector_5}
   ]
  }
  MaterialExpressionMultiply_2 {
   class=MaterialExpressionMultiply
   inputs [
    A {from=MaterialExpressionSceneTexelSize_7}
    B {from=MaterialExpressionConstant2Vector_6}
   ]
  }
  MaterialExpressionMultiply_5 {
   class=MaterialExpressionMultiply
   inputs [
    A {from=MaterialExpressionSceneTexelSize_7}
    B {from=MaterialExpressionConstant2Vector_7}
   ]
  }
  MaterialExpressionMultiply_6 {
   class=MaterialExpressionMultiply
   inputs [
    A {from=MaterialExpressionSceneTexelSize_7}
    B {from=MaterialExpressionConstant2Vector_8}
   ]
  }
  MaterialExpressionPixelDepth_1 {class=MaterialExpressionPixelDepth, inputs [], properties {}}
  MaterialExpressionReroute_3 {class=MaterialExpressionReroute, inputs [None {from=MaterialExpressionSceneTexture_3, output=Color}]}
  MaterialExpressionSaturate_1 {class=MaterialExpressionSaturate, inputs [None {from=MaterialExpressionSign_1}]}
  MaterialExpressionSceneTexelSize_7 {class=MaterialExpressionSceneTexelSize, inputs [], properties {}}
  MaterialExpressionSceneTexture_0 {class=MaterialExpressionSceneTexture, inputs [UVs {from=MaterialExpressionAdd_51, output=""}], properties {sceneTextureId=PPI_SceneDepth}}
  MaterialExpressionSceneTexture_1 {class=MaterialExpressionSceneTexture, inputs [UVs {from=MaterialExpressionAdd_1, output=""}], properties {sceneTextureId=PPI_SceneDepth}}
  MaterialExpressionSceneTexture_2 {class=MaterialExpressionSceneTexture, inputs [UVs {from=MaterialExpressionAdd_2, output=""}], properties {sceneTextureId=PPI_SceneDepth}}
  MaterialExpressionSceneTexture_3 {class=MaterialExpressionSceneTexture, inputs [UVs {from=MaterialExpressionAdd_3, output=""}], properties {sceneTextureId=PPI_SceneDepth}}
  MaterialExpressionScreenPosition_0 {class=MaterialExpressionScreenPosition, inputs [], properties {}}
  MaterialExpressionSign_1 {class=MaterialExpressionSign, inputs [None {from=MaterialExpressionMin_5, output=""}], properties {}}
  MaterialExpressionSubtract_2 {
   class=MaterialExpressionSubtract
   inputs [
    A {from=MaterialExpressionConvert_4}
    B {from=MaterialExpressionConvert_3}
   ]
  }
 }}
