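# Material function that computes screen-space refracted viewport UVs from a tangent-space normal and IOR, approximating built-in refraction.
# The view-space XY difference between vertex and pixel normals becomes a thresholded offset, scaled by IOR, distortion scale and FOV correction.
# An optional depth contribution scales distortion by distance to the refracted scene over pixel depth.
# A static-switch mask, bypassed on mobile, multi-samples SceneDepth to reject distortion of foreground pixels.
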MaterialFunction /Game/Global/MaterialFunctions/TechArt/Refraction/MF_RefractionDistortion_ScreenSpace
graph {nodes [
  n1 = NamedRerouteUsage()
  n2 = MaterialFunctionCall!(Distorted Viewport UV=@n1) {desc=Reject pixels in front of the refractive surface, materialFunction=&/MF_RefractionMask}
  n3 = LinearInterpolate(A=NamedRerouteUsage(), B=@n1, Alpha=@n2."Distortion Mask (S)")
  n4 = NamedRerouteUsage()
  n5 = FunctionInput(Preview=StaticBool()) {
   bUsePreviewValueAsDefault=true
   description="Enable SceneDepth multi-sampling. If any sample is closer than PixelDepth, refraction is disabled for the entire pixel. Prevents surfaces closer to the camera than the current refractive interface to get incorrectly distorted. Samples SceneDepth four times. Default : OFF"
   inputName=Enable Refraction Mask
   inputType=FunctionInput_StaticBool
   sortPriority=50
  }
  FunctionOutput(None=ShadingPathSwitch(Default=StaticSwitch(True=@n3, False=@n4, Value=@n5), Mobile=NamedRerouteUsage())) {outputName="Refracted Viewport UV (V2)", sortPriority=10}
  n6 = FunctionInput() {
   bUsePreviewValueAsDefault=true
   description="Minimum difference between the vertex and pixel normals to produce distortion. Can be useful if the normal map is not very clean where it's supposed to be smooth (eg. 8bit artefacts)"
   inputName=Pixel Normal Offset Threshold
   inputType=FunctionInput_Scalar
   sortPriority=40
  }
  n7 = Multiply(A=Max(A=@n6) {constB=0}) {constB=0.01}
  n8 = Abs(None=NamedRerouteUsage())
  NamedRerouteDeclaration(None=Multiply(A=NamedRerouteUsage(), B=SmoothStep(Max=@n7, Value=@n8))) {name="Pixel Normal Offset (Smooth) (V2)", nodeColor=<rgba 1 0.3047 0 1>}
  n9 = Multiply(A=Multiply(A=NamedRerouteUsage(), B=NamedRerouteUsage()), B=NamedRerouteUsage()) {desc=Distortion Intensity}
  n10 = ScreenPosition()
  NamedRerouteDeclaration(None=Add(
    A=Multiply(A=@n9) {constB=1.5, desc="Correction factor for\r\nViewportUV distortion\r\nto match builtin refraction"}
    B=NamedRerouteDeclaration(None=@n10.ViewportUV) {name="Viewport UV (V2)", nodeColor=<rgba 0 0.1094 1 1>}
   )) {name="Refracted Viewport UV (V2)", nodeColor=<rgba 1 0.2109 0 1>}
  n11 = MaterialFunctionCall!(Viewport UV, Max Distance) {materialFunction=&/MF_DistanceToRefractedScene}
  n12 = PixelDepth()
  n13 = FunctionInput() {
   bUsePreviewValueAsDefault=true
   description="How much depth should affect distortion scale. 0: distortion amount is constant regardless of depth; 1: far objects are more distorted than near ones (more accurate but can lead to extreme distortion)"
   inputName=Depth Contribution
   inputType=FunctionInput_Scalar
   sortPriority=30
   previewValue=<1 0 0 1>
  }
  NamedRerouteDeclaration(None=LinearInterpolate(B=Divide(A=@n11."Distance (S)", B=@n12.R), Alpha=@n13) {constA=1}) {name="Depth-Based Scale (S)", nodeColor=<rgba 0.2188 0 1 1>}
  n14 = ComponentMask(None=Transform(None=Constant3Vector() {desc="Vertex Normal (tangent space)", constant=<rgba 0 0 1 0>}) {transformType=TRANSFORM_View}) {g=true, r=true}
  n15 = ComponentMask(None=Transform(None=FunctionInput() {description=Normalized pixel normal in tangent space, inputName=Normal TS, previewValue=<0 0 1 1>}) {transformType=TRANSFORM_View}) {g=true, r=true}
  NamedRerouteDeclaration(None=Multiply(A=Subtract(A=@n14, B=@n15) {desc=Pixel Normal Offset}, B=NamedRerouteUsage())) {name="Pixel Normal Offset (V2)", nodeColor=<rgba 0.1016 1 0 1>}
  n16 = ViewSize()
  n17 = Constant() {desc=OffsetFudgeFactor, r=0.0002}
  n18 = AppendVector(A=@n17, B=Multiply(A=@n17) {constB=-1})
  NamedRerouteDeclaration(None=Multiply(A=Multiply(A=@n16, B=@n18), B=ViewProperty() {property=MEVP_TanHalfFieldOfView}.InvProperty) {desc=FOV Fix}) {name="FOV Fix (S)", nodeColor=<rgba 1 0 0.6562 1>}
  n19 = FunctionInput() {bUsePreviewValueAsDefault=true, inputName=Distortion Scale, inputType=FunctionInput_Scalar, sortPriority=20, previewValue=<1 0 0 1>}
  n20 = FunctionInput() {
   description="Index of refraction of the transmissive dielectric interface; Valid range : [1.0, 2.0]"
   inputName=Refraction IOR
   inputType=FunctionInput_Scalar
   sortPriority=10
   previewValue=<1 0 0 1>
  }
  n21 = Subtract(A=@n20)
  n22 = OneMinus(None=Saturate(None=@n21))
  n23 = Add(A=Multiply(A=Multiply(A=@n22, B=@n22)) {constB=0.5}) {constB=0.5, desc="Reduce distortion by\r\n50% when IOR = 2"}
  NamedRerouteDeclaration(None=Multiply(A=@n19, B=Multiply(A=@n21, B=@n23)) {desc=Pixel Normal Offset Factor}) {name="Pixel Normal Offset Factor (S)", nodeColor=<rgba 0 0.9766 1 1>}
 ]}
