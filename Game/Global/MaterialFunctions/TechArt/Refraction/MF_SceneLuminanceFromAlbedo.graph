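# Material function that estimates the luminance of a refracted scene from its albedo, for simulated colored transmittance when SceneColor is not sampled.
# Sun illuminance weighted by saturated N·L is added to sky illuminance, which defaults to a fully rough skylight environment sample times Pi.
# That total illuminance is multiplied by Lambertian diffuse (albedo over Pi) and tinted by a scene color scale akin to SingleLayerWater's ColorScaleBehindWater.
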MaterialFunction /Game/Global/MaterialFunctions/TechArt/Refraction/MF_SceneLuminanceFromAlbedo
graph {nodes [
  n1 = FunctionInput() {
   bUsePreviewValueAsDefault=true
   description=Tints the refracted SceneColor. Equivalent to SingleLayerWater's ColorScaleBehindWater parameter
   inputName=Scene Color Scale
   sortPriority=50
   previewValue=<1 1 1 1>
  }
  n2 = FunctionInput() {
   bUsePreviewValueAsDefault=true
   description="Should Ideally be the average diffuse albedo of the refracted scene; Only used for simulated colored transmittance when not sampling SceneColor"
   inputName=Scene Albedo
   sortPriority=40
   previewValue=<0.18 0.18 0.18 1>
  }
  n3 = Divide(A=Saturate(None=@n2), B=MaterialFunctionCall() {materialFunction=&Pi}.Result) {desc=Lambertian Diffuse}
  n4 = NamedRerouteUsage()
  FunctionOutput(None=Multiply(A=@n1, B=Multiply(A=@n3, B=@n4))) {outputName="Scene Luminance (V3)"}
  n5 = FunctionInput() {
   bUsePreviewValueAsDefault=true
   description=Only used for simulated colored transmittance when not sampling SceneColor
   inputName=Sun Illuminance
   sortPriority=10
   previewValue=<1 1 1 1>
  }
  n6 = FunctionInput() {bUsePreviewValueAsDefault=true, inputName=Sun Direction, sortPriority=20, previewValue=<1 1 1 1>}
  n7 = FunctionInput(Preview=VertexNormalWS()) {
   bUsePreviewValueAsDefault=true
   description=Provided normal must be normalized
   inputName=Scene Normal
   sortPriority=30
   previewValue=<0 0 1 1>
  }
  n8 = Saturate(None=DotProduct(A=@n6, B=@n7))
  n9 = Constant() {r=1}
  n10 = FunctionInput(Preview=Multiply(
    A=SkyLightEnvMapSample(Direction=@n7, Roughness=@n9)
    B=MaterialFunctionCall() {materialFunction=&Pi}.Result
   ) {desc=Luminance to illuminance}) {
   bUsePreviewValueAsDefault=true
   description=Only used for simulated colored transmittance when not sampling SceneColor
   inputName=Sky Illuminance
   previewValue=<1 1 1 1>
  }
  NamedRerouteDeclaration(None=Add(A=Multiply(A=@n5, B=@n8), B=@n10)) {name="Sun && Sky Illuminance (V3)", nodeColor=<rgba 0.4062 0 1 1>}
 ]}
