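# Material function that estimates a normalized thickness scalar from view angle, filed under the refraction tech-art functions.
# It dots the negated normal with an inward direction, defaulting to vertex normal and negated camera vector, then saturates it for a Fresnel-like facing ratio.
# That ratio is raised to an exponent, scaled by Parallel Intensity, offset by Perpendicular Intensity as the grazing value, then divided by Perpendicular Intensity plus one.
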MaterialFunction /Game/Global/MaterialFunctions/TechArt/Refraction/MF_ThicknessSimple
graph {nodes [
  n1 = FunctionInput(Preview=VertexNormalWS()) {bUsePreviewValueAsDefault=true, description=Default : VertexNormalWS, inputName=Normal, previewValue=<0 0 1 1>}
  n2 = FunctionInput(Preview=Multiply(A=CameraVectorWS()) {constB=-1}) {
   bUsePreviewValueAsDefault=true
   description="Pointing inside the surface; Default : CameraVectorWS"
   inputName=Direction
   sortPriority=10
   previewValue=<0 0 1 1>
  }
  n3 = DotProduct(A=Multiply(A=@n1) {constB=-1}, B=@n2)
  n4 = FunctionInput() {bUsePreviewValueAsDefault=true, inputName=Facing Ratio Exp, inputType=FunctionInput_Scalar, sortPriority=20, previewValue=<0.9 0 0 1>}
  n5 = FunctionInput() {bUsePreviewValueAsDefault=true, inputName=Parallel Intensity, inputType=FunctionInput_Scalar, sortPriority=30, previewValue=<0.4 0 0 1>}
  n6 = Multiply(A=Power(Base=Saturate(None=@n3), Exp=@n4) {desc=Fresnel Exponent}, B=@n5) {desc=Facing Value}
  n7 = FunctionInput() {bUsePreviewValueAsDefault=true, inputName=Perpendicular Intensity, inputType=FunctionInput_Scalar, sortPriority=40, previewValue=<0.7 0 0 1>}
  FunctionOutput(None=Divide(A=Add(A=@n6, B=@n7), B=Add(A=@n7)) {desc=Grazing Angle Value}) {outputName="Thickness Normalized (S)"}
 ]}
